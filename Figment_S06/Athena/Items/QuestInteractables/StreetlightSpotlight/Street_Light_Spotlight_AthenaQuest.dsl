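; Fortnite Athena quest streetlight that detects players emoting beneath it, grants quest progress outside Creative, and presents a disco spotlight with music.
; The server tracks pawns in a capsule, polls them on a timer for emotes, and sets replicated DiscoLightsOn.
; Client-side logic cycles light and material colours and uses a timeline to open the music's low-pass filter and widen attenuation falloff.
; Clients play success FX on quest completion, binding updates after a delay that waits for the quest manager.
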(event EventBeginPlay
  (bind _returnvalue (Fortnite|GetLocalFortPlayerControllers true))
  (bind _output (Utilities|Array|Get(acopy) _returnvalue))
  (bind _dancecollisioncapsule (Variables|Default|GetDanceCollisionCapsule))
  (bind _streetprop_streetlighturban_04 (Variables|Default|GetStreetProp_StreetLightUrban_04))
  ;[1 1 1] Check if In Creative Mode or not
  (bind _asfort_game_state_athena (Utilities|Casting|CastToFortGameStateAthena (Game|GetGameState)))
  (if (Variables|/Script/FortniteGame.FortGameStateAthena|GetbPlaylistDataIsLoaded _asfort_game_state_athena)
    (if (GameplayTags|HasTag (Class|FortGameStateAthena|GetAthenaPlaylistContextTags _asfort_game_state_athena) "(TagName=\"Athena.Creative\")")
      (Variables|Default|SetIsCreativeMode true)
      (else
        (Variables|Default|SetIsCreativeMode false)))
    (Networking|SwitchHasAuthority
      (:Authority
        ;[1 1 1] SERVER: Bind Overlaps for Pawn Collision under Streetlights
        (Collision|BindEventtoOnComponentBeginOverlap _dancecollisioncapsule (EventDispatchers|CreateEvent :!event Custom|PlayerEnteredDanceArea))
        (Collision|BindEventtoOnComponentEndOverlap _dancecollisioncapsule (EventDispatchers|CreateEvent :!event Custom|PlayerLeftDanceArea)))
      (:Remote
        ;[1 0.5334 0] Hack to Wait for Quest Manager to Load
        (Utilities|FlowControl|Delay 5.0)
        (Utilities|FlowControl|Sequence
          (:then_0
            ;[1 1 1] CLIENT: Bind Event for When Quest Gets Updated
            (bind _asfort_player_controller_athena (Utilities|Casting|CastToFortPlayerControllerAthena _output))
            (bind (_questisvalid _questcompleted _objectivecompleted) (Default|AthenaCheckQuestandObjectiveCompletion _output (Variables|Default|GetQuestItem) (Variables|Default|GetObjectiveBackendName)))
            (if _questisvalid
              (if _objectivecompleted
                _
                (elif (Variables|Default|GetIsCreativeMode)
                  _
                  (else
                    (Quests|BindEventtoOnQuestsUpdated (Quests|GetQuestManager _asfort_player_controller_athena "Invalid") (EventDispatchers|CreateEvent :!event Custom|QuestUpdated))
                    (Variables|Default|SetQuestLocallyActive true))))))
          (:then_1
            ;[1 1 1] Start w/ the LPF closed.
            (CallFunction|LPFTransition false)
            ;[1 1 1] DMI's for Light mats
            (bind _returnvalue_2 (Rendering|Material|CreateDynamicMaterialInstance _streetprop_streetlighturban_04 0 "/Game/Environments/Props/Street/Materials/PP2/MI_StreetLightUrban_04_Ambient.MI_StreetLightUrban_04_Ambient" "None" :!class /Script/Engine.PrimitiveComponent))
            (Variables|Default|SetLampMID _returnvalue_2)
            (bind _returnvalue_3 (Rendering|Material|CreateDynamicMaterialInstance _streetprop_streetlighturban_04 1 "/Game/Environments/Props/Street/Materials/PP2/MI_StreetLight_04_ambient_Cone.MI_StreetLight_04_ambient_Cone" "None" :!class /Script/Engine.PrimitiveComponent))
            (Variables|Default|SetConeMID _returnvalue_3)))))
    (else
      (Playlist|BindEventtoOnCurrentPlaylistLoaded _asfort_game_state_athena (delegate Custom|PlaylistLoaded)))))

(event Custom|PlaylistLoaded
  (params (PlaylistName :type Name) (PlaylistContextTags :type Gameplay%20Tag%20Container%20Structure%20%28by%20ref%29))
  (bind _returnvalue (Fortnite|GetLocalFortPlayerControllers true))
  (bind _output (Utilities|Array|Get(acopy) _returnvalue))
  (bind _dancecollisioncapsule (Variables|Default|GetDanceCollisionCapsule))
  (bind _streetprop_streetlighturban_04 (Variables|Default|GetStreetProp_StreetLightUrban_04))
  ;[1 1 1] Check if In Creative Mode or not
  (if (GameplayTags|HasTag PlaylistContextTags "(TagName=\"Athena.Creative\")")
    (Variables|Default|SetIsCreativeMode true)
    (else
      (Variables|Default|SetIsCreativeMode false)))
  (Networking|SwitchHasAuthority
    (:Authority
      ;[1 1 1] SERVER: Bind Overlaps for Pawn Collision under Streetlights
      (Collision|BindEventtoOnComponentBeginOverlap _dancecollisioncapsule (EventDispatchers|CreateEvent :!event Custom|PlayerEnteredDanceArea))
      (Collision|BindEventtoOnComponentEndOverlap _dancecollisioncapsule (EventDispatchers|CreateEvent :!event Custom|PlayerLeftDanceArea)))
    (:Remote
      ;[1 0.5334 0] Hack to Wait for Quest Manager to Load
      (Utilities|FlowControl|Delay 5.0)
      (Utilities|FlowControl|Sequence
        (:then_0
          ;[1 1 1] CLIENT: Bind Event for When Quest Gets Updated
          (bind _asfort_player_controller_athena (Utilities|Casting|CastToFortPlayerControllerAthena _output))
          (bind (_questisvalid _questcompleted _objectivecompleted) (Default|AthenaCheckQuestandObjectiveCompletion _output (Variables|Default|GetQuestItem) (Variables|Default|GetObjectiveBackendName)))
          (if _questisvalid
            (if _objectivecompleted
              _
              (elif (Variables|Default|GetIsCreativeMode)
                _
                (else
                  (Quests|BindEventtoOnQuestsUpdated (Quests|GetQuestManager _asfort_player_controller_athena "Invalid") (EventDispatchers|CreateEvent :!event Custom|QuestUpdated))
                  (Variables|Default|SetQuestLocallyActive true))))))
        (:then_1
          ;[1 1 1] Start w/ the LPF closed.
          (CallFunction|LPFTransition false)
          ;[1 1 1] DMI's for Light mats
          (bind _returnvalue_2 (Rendering|Material|CreateDynamicMaterialInstance _streetprop_streetlighturban_04 0 "/Game/Environments/Props/Street/Materials/PP2/MI_StreetLightUrban_04_Ambient.MI_StreetLightUrban_04_Ambient" "None" :!class /Script/Engine.PrimitiveComponent))
          (Variables|Default|SetLampMID _returnvalue_2)
          (bind _returnvalue_3 (Rendering|Material|CreateDynamicMaterialInstance _streetprop_streetlighturban_04 1 "/Game/Environments/Props/Street/Materials/PP2/MI_StreetLight_04_ambient_Cone.MI_StreetLight_04_ambient_Cone" "None" :!class /Script/Engine.PrimitiveComponent))
          (Variables|Default|SetConeMID _returnvalue_3))))))

(event Custom|QuestUpdated
  ;[1 1 1] CLIENT: After Quest Update - check Local Clients to give rewards
  (Development|PrintString "QuestUpdated" false false "(R=0.309646,G=1.000000,B=0.000000,A=1.000000)" 5.0)
  (if (Variables|Default|GetQuestLocallyActive)
    (bind (_questisvalid _questcompleted _objectivecompleted) (Default|AthenaCheckQuestandObjectiveCompletion (Utilities|Array|Get(acopy) (Fortnite|GetLocalFortPlayerControllers true)) (Variables|Default|GetQuestItem) (Variables|Default|GetObjectiveBackendName)))
    (if (or _questcompleted _objectivecompleted)
      (Quests|UnbindEventfromOnQuestsUpdated (Quests|GetQuestManager (stale EventBeginPlay _asfort_player_controller_athena) "Invalid") (EventDispatchers|CreateEvent :!event Custom|QuestUpdated))
      (Variables|Default|SetQuestLocallyActive false)
      (Utilities|FlowControl|Delay 0.2)
      (CallFunction|PlayQuestSuccessFX))
    (else
      (Development|PrintString "Not Locally Active" false false "(R=1.000000,G=0.075957,B=0.000000,A=1.000000)" 5.0))))

(event Custom|PlayerEnteredDanceArea
  (params (OverlappedComponent :type Primitive%20Component%20Object%20Reference) (OtherActor :type Actor%20Object%20Reference) (OtherComp :type Primitive%20Component%20Object%20Reference) (OtherBodyIndex :type Integer) (bFromSweep :type Boolean) (SweepResult :type Hit%20Result%20Structure%20%28by%20ref%29))
  (bind _playersindancearea (Variables|Default|GetPlayersInDanceArea))
  ;[1 1 1] SERVER: Add/Remove Pawns who Enter/Exit dance zone under light - Start/End Timer that checks for dancers
  (bind _asfort_pawn (Utilities|Casting|CastToFortPawn OtherActor))
  (Utilities|Array|AddUnique _playersindancearea _asfort_pawn)
  (if (== (Utilities|Array|Length _playersindancearea) 1)
    (bind _returnvalue (Utilities|Time|SetTimerbyEvent (EventDispatchers|CreateEvent :!event Custom|CheckForDanceTimer) 0.2 true))
    (Variables|Default|SetCheckforDancingTimer _returnvalue)))

(event Custom|PlayQuestSuccessFX
  (bind _returnvalue (Fortnite|GetLocalFortPlayerControllers true))
  (bind _output (Utilities|Array|Get(acopy) _returnvalue))
  (bind _returnvalue_2 (Game|Player|GetPlayerPawn _output))
  ;[1 1 1] Quest Success FX
  (Development|PrintString "SUCCESS FX" false false "(R=1.000000,G=0.703857,B=0.040648,A=1.000000)" 5.0)
  (Effects|SpawnFXSystemAttached (Variables|Default|GetQuestSuccessEmitter) (Variables|/Script/Engine.Character|GetMesh _returnvalue_2))
  (Audio|PlaySoundatLocation (Variables|Default|GetQuestSuccessSound) (Transformation|GetActorLocation _returnvalue_2)) ; Quest Success Sound
)

(event Custom|PlayerLeftDanceArea
  (params (OverlappedComponent :type Primitive%20Component%20Object%20Reference) (OtherActor :type Actor%20Object%20Reference) (OtherComp :type Primitive%20Component%20Object%20Reference) (OtherBodyIndex :type Integer))
  (bind _playersindancearea (Variables|Default|GetPlayersInDanceArea))
  ;[1 1 1] SERVER: Add/Remove Pawns who Enter/Exit dance zone under light - Start/End Timer that checks for dancers
  (bind _asfort_pawn (Utilities|Casting|CastToFortPawn OtherActor))
  (Utilities|Array|RemoveItem _playersindancearea _asfort_pawn)
  (if (< (Utilities|Array|Length _playersindancearea) 1)
    (Utilities|Time|ClearandInvalidateTimerbyHandle (Variables|Default|GetCheckforDancingTimer))))

(event Custom|CheckForDanceTimer
  ;[1 1 1] SERVER: Check for Dancing Pawns under Streetlight
  (bind _array_element (Utilities|Array|ForEachLoopwithBreak (Variables|Default|GetPlayersInDanceArea))
    (:LoopBody
      (bind _asathena_player_controller (Utilities|Casting|CastToAthena_PlayerController (Pawn|GetController _array_element)))
      (if (Emote|IsPlayingEmote _asathena_player_controller)
        (Variables|Default|SetFoundDancer true)
        (Utilities|FlowControl|Sequence
          (:then_0
            _)
          (:then_1
            _))
        ;[1 1 1] SERVER: Give Quest Reward if not Creative Mode
        (if (not (Variables|Default|GetIsCreativeMode))
          (Class|AthenaQuestFunctionLibrary|AthenaGiveQuestUpdatetoPlayers (Variables|Default|GetQuestItem) (Variables|Default|GetObjectiveBackendName) (Variables|Default|GetObjectiveStat) (Utilities|Array|MakeArray (Fortnite|GetFortPlayerControllerfromActor _asathena_player_controller)))))
      (Variables|Default|SetFoundDancer false))
    (:Completed
      _))
  (if (Variables|Default|GetFoundDancer)
    (Variables|Default|SetDiscoLightsOn true)
    (else
      (Variables|Default|SetDiscoLightsOn false)
      (CallFunction|ClearDiscoLightTimer)))
  (Utilities|FlowControl|DoOnce false
    (:Completed
      (CallFunction|DiscoLightTimer))))

(event Custom|LPF%20Transition
  (params (Is%20LPF%20Open :type Boolean))
  (bind (_attenuationoverrides_battenuate _attenuationoverrides_bspatialize _attenuationoverrides_battenuatewithlpf _attenuationoverrides_benablelistenerfocus _attenuationoverrides_benablefocusinterpolation _attenuationoverrides_benableocclusion _attenuationoverrides_busecomplexcollisionforocclusion _attenuationoverrides_benablereverbsend _attenuationoverrides_benablepriorityattenuation _attenuationoverrides_bapplynormalizationtostereosounds _attenuationoverrides_benablelogfrequencyscaling _attenuationoverrides_benablesubmixsends _attenuationoverrides_benablesourcedataoverride _attenuationoverrides_benablesendtoaudiolink _attenuationoverrides_spatializationalgorithm _attenuationoverrides_audiolinksettingsoverride _attenuationoverrides_binauralradius _attenuationoverrides_customlowpassairabsorptioncurve _attenuationoverrides_customhighpassairabsorptioncurve _attenuationoverrides_absorptionmethod _attenuationoverrides_occlusiontracechannel _attenuationoverrides_reverbsendmethod _attenuationoverrides_priorityattenuationmethod _attenuationoverrides_nonspatializedradiusstart _attenuationoverrides_nonspatializedradiusend _attenuationoverrides_nonspatializedradiusmode _attenuationoverrides_stereospread _attenuationoverrides_lpfradiusmin _attenuationoverrides_lpfradiusmax _attenuationoverrides_lpffrequencyatmin _attenuationoverrides_lpffrequencyatmax _attenuationoverrides_hpffrequencyatmin _attenuationoverrides_hpffrequencyatmax _attenuationoverrides_focusazimuth _attenuationoverrides_nonfocusazimuth _attenuationoverrides_focusdistancescale _attenuationoverrides_nonfocusdistancescale _attenuationoverrides_focuspriorityscale _attenuationoverrides_nonfocuspriorityscale _attenuationoverrides_focusvolumeattenuation _attenuationoverrides_nonfocusvolumeattenuation _attenuationoverrides_focusattackinterpspeed _attenuationoverrides_focusreleaseinterpspeed _attenuationoverrides_occlusionlowpassfilterfrequency _attenuationoverrides_occlusionvolumeattenuation _attenuationoverrides_occlusioninterpolationtime _attenuationoverrides_reverbwetlevelmin _attenuationoverrides_reverbwetlevelmax _attenuationoverrides_reverbdistancemin _attenuationoverrides_reverbdistancemax _attenuationoverrides_manualreverbsendlevel _attenuationoverrides_priorityattenuationmin _attenuationoverrides_priorityattenuationmax _attenuationoverrides_priorityattenuationdistancemin _attenuationoverrides_priorityattenuationdistancemax _attenuationoverrides_manualpriorityattenuation _attenuationoverrides_customreverbsendcurve _attenuationoverrides_submixsendsettings _attenuationoverrides_custompriorityattenuationcurve _attenuationoverrides_pluginsettings _attenuationoverrides_distancealgorithm _attenuationoverrides_attenuationshape _attenuationoverrides_falloffmode _attenuationoverrides_dbattenuationatmax _attenuationoverrides_attenuationshapeextents _attenuationoverrides_coneoffset _attenuationoverrides_falloffdistance _attenuationoverrides_conesphereradius _attenuationoverrides_conespherefalloffdistance _attenuationoverrides_customattenuationcurve) (Utilities|Struct|BreakSoundAttenuationSettings (Variables|/Script/Engine.AudioComponent|GetAttenuationOverrides_bAttenuate (Variables|Default|GetMusicAudioLPF))))
  (bind _musicaudiolpf (Variables|Default|GetMusicAudioLPF))
  ;[1 1 1] LPF + Extend audio when dancing under the light
  (Audio|PlaySoundatLocation (select Is%20LPF%20Open "/Figment_S06/Athena/Sounds/Quests/StreetlightMusic_LPF_Open_Cue.StreetlightMusic_LPF_Open_Cue" "/Figment_S06/Athena/Sounds/Quests/StreetlightMusic_LPF_Close_Cue.StreetlightMusic_LPF_Close_Cue") (Transformation|GetWorldLocation (Variables|Default|GetMusicAudioLPF)))
  (Audio|Components|Audio|SetLowPassFilterEnabled (Variables|Default|GetMusicAudioLPF) true :!class /Script/Engine.AudioComponent)
  (Components|Timeline|SetPlayRate (Variables|Street_Light_Spotlight_AthenaQuest|GetLPF%20Transition%20Timeline) (/ 1.0 2.0))
  (if Is%20LPF%20Open
    (bind (_direction _cutoff _attenuation) (Timeline|LPF%20Transition%20Timeline Play)
      (:Update
        (Audio|Components|Audio|SetLowPassFilterFrequency (Variables|Default|GetMusicAudioLPF) (Math|Float|MapRangeUnclamped _cutoff 0.0 1.0 5e+02 2e+04) :!class /Script/Engine.AudioComponent)
        ;; Nasty Attenuation Setting
        (bind _output_get (Variables|/Script/Engine.AudioComponent|SetAttenuationOverrides_bAttenuate (Utilities|Struct|MakeSoundAttenuationSettings :bAttenuate _attenuationoverrides_battenuate :bSpatialize _attenuationoverrides_bspatialize :bAttenuateWithLPF _attenuationoverrides_battenuatewithlpf :bEnableListenerFocus _attenuationoverrides_benablelistenerfocus :bEnableFocusInterpolation _attenuationoverrides_benablefocusinterpolation :bEnableOcclusion _attenuationoverrides_benableocclusion :bUseComplexCollisionForOcclusion _attenuationoverrides_busecomplexcollisionforocclusion :bEnableReverbSend _attenuationoverrides_benablereverbsend :bApplyNormalizationToStereoSounds _attenuationoverrides_bapplynormalizationtostereosounds :bEnableLogFrequencyScaling _attenuationoverrides_benablelogfrequencyscaling :SpatializationAlgorithm _attenuationoverrides_spatializationalgorithm :CustomLowpassAirAbsorptionCurve _attenuationoverrides_customlowpassairabsorptioncurve :CustomHighpassAirAbsorptionCurve _attenuationoverrides_customhighpassairabsorptioncurve :AbsorptionMethod _attenuationoverrides_absorptionmethod :OcclusionTraceChannel _attenuationoverrides_occlusiontracechannel :ReverbSendMethod _attenuationoverrides_reverbsendmethod :StereoSpread _attenuationoverrides_stereospread :LPFRadiusMin _attenuationoverrides_lpfradiusmin :LPFRadiusMax _attenuationoverrides_lpfradiusmax :LPFFrequencyAtMin _attenuationoverrides_lpffrequencyatmin :LPFFrequencyAtMax _attenuationoverrides_lpffrequencyatmax :HPFFrequencyAtMin _attenuationoverrides_hpffrequencyatmin :HPFFrequencyAtMax _attenuationoverrides_hpffrequencyatmax :FocusAzimuth _attenuationoverrides_focusazimuth :NonFocusAzimuth _attenuationoverrides_nonfocusazimuth :FocusDistanceScale _attenuationoverrides_focusdistancescale :NonFocusDistanceScale _attenuationoverrides_nonfocusdistancescale :FocusPriorityScale _attenuationoverrides_focuspriorityscale :NonFocusPriorityScale _attenuationoverrides_nonfocuspriorityscale :FocusVolumeAttenuation _attenuationoverrides_focusvolumeattenuation :NonFocusVolumeAttenuation _attenuationoverrides_nonfocusvolumeattenuation :FocusAttackInterpSpeed _attenuationoverrides_focusattackinterpspeed :FocusReleaseInterpSpeed _attenuationoverrides_focusreleaseinterpspeed :OcclusionLowPassFilterFrequency _attenuationoverrides_occlusionlowpassfilterfrequency :OcclusionVolumeAttenuation _attenuationoverrides_occlusionvolumeattenuation :OcclusionInterpolationTime _attenuationoverrides_occlusioninterpolationtime :ReverbWetLevelMin _attenuationoverrides_reverbwetlevelmin :ReverbWetLevelMax _attenuationoverrides_reverbwetlevelmax :ReverbDistanceMin _attenuationoverrides_reverbdistancemin :ReverbDistanceMax _attenuationoverrides_reverbdistancemax :ManualReverbSendLevel _attenuationoverrides_manualreverbsendlevel :CustomReverbSendCurve _attenuationoverrides_customreverbsendcurve :PluginSettings _attenuationoverrides_pluginsettings :DistanceAlgorithm _attenuationoverrides_distancealgorithm :AttenuationShape _attenuationoverrides_attenuationshape :dBAttenuationAtMax _attenuationoverrides_dbattenuationatmax :AttenuationShapeExtents _attenuationoverrides_attenuationshapeextents :ConeOffset _attenuationoverrides_coneoffset :FalloffDistance (Math|Float|MapRangeUnclamped _attenuation 0.0 1.0 1e+02 5e+03) :CustomAttenuationCurve _attenuationoverrides_customattenuationcurve) _musicaudiolpf))
        (Audio|Components|Audio|AdjustAttenuation _musicaudiolpf _output_get)))
    (else
      (Timeline|LPF%20Transition%20Timeline Reverse))))

(event Custom|DiscoLightTimer
  ;[1 1 1] CLIENT: Rotate through Disco Lights
  (Networking|SwitchHasAuthority
    (:Remote
      (bind _returnvalue (Utilities|Time|SetTimerbyEvent (delegate Custom|RotateThroughLightColors) (Variables|Default|GetLightSwitchTime) true))
      (Variables|Default|SetLightSwitchTimer _returnvalue)
      (Rendering|Components|Light|SetIntensity (Variables|Default|GetSpotLight) 1.5e+02)
      (Rendering|Components|Light|SetLightColor (Variables|Default|GetSpotLight) "(R=0.339420,G=0.000000,B=0.396755,A=1.000000)" true)
      (Rendering|Material|SetVectorParameterValue (Variables|Default|GetLampMID) "Emissive Multiplier" "(R=1.000000,G=0.000000,B=0.889239,A=1.000000)" :!class /Script/Engine.MaterialInstanceDynamic)
      (Rendering|Material|SetVectorParameterValue (Variables|Default|GetConeMID) "Emissive Color" "(R=1.000000,G=0.000000,B=0.889239,A=1.000000)" :!class /Script/Engine.MaterialInstanceDynamic))))

(event Custom|RotateThroughLightColors
  (bind _lightcolorrotation (Variables|Default|GetLightColorRotation))
  (bind _spotlight (Variables|Default|GetSpotLight))
  (bind _conemid (Variables|Default|GetConeMID))
  (bind _lampmid (Variables|Default|GetLampMID))
  ;[1 1 1] CLIENT: Rotate through Disco Lights
  (Math|Integer|IncrementInt _lightcolorrotation
    (:%20%20
      (switch int _lightcolorrotation
        (:1
          (Rendering|Components|Light|SetLightColor _spotlight "(R=0.000000,G=0.393004,B=0.396755,A=1.000000)" true)
          (Rendering|Material|SetVectorParameterValue _lampmid "Emissive Multiplier" "(R=0.000000,G=0.852744,B=1.000000,A=1.000000)" :!class /Script/Engine.MaterialInstanceDynamic)
          (Rendering|Material|SetVectorParameterValue _conemid "Emissive Color" "(R=0.000000,G=0.852744,B=1.000000,A=1.000000)" :!class /Script/Engine.MaterialInstanceDynamic))
        (:2
          (Rendering|Components|Light|SetLightColor _spotlight "(R=0.396755,G=0.322566,B=0.000000,A=1.000000)" true)
          (Rendering|Material|SetVectorParameterValue _lampmid "Emissive Multiplier" "(R=1.000000,G=0.782032,B=0.000000,A=1.000000)" :!class /Script/Engine.MaterialInstanceDynamic)
          (Rendering|Material|SetVectorParameterValue _conemid "Emissive Color" "(R=1.000000,G=0.782032,B=0.000000,A=1.000000)" :!class /Script/Engine.MaterialInstanceDynamic))
        (:3
          (Variables|Default|SetLightColorRotation 0)
          (Rendering|Components|Light|SetLightColor _spotlight "(R=0.339420,G=0.000000,B=0.396755,A=1.000000)" true)
          (Rendering|Material|SetVectorParameterValue _lampmid "Emissive Multiplier" "(R=1.000000,G=0.000000,B=0.889239,A=1.000000)" :!class /Script/Engine.MaterialInstanceDynamic)
          (Rendering|Material|SetVectorParameterValue _conemid "Emissive Color" "(R=1.000000,G=0.000000,B=0.889239,A=1.000000)" :!class /Script/Engine.MaterialInstanceDynamic))))))

(event Custom|ClearDiscoLightTimer
  ;[1 1 1] CLIENT: clear out Disco Light Timer
  (Networking|SwitchHasAuthority
    (:Remote
      (Variables|Default|SetLightColorRotation 0)
      (Utilities|Time|ClearandInvalidateTimerbyHandle (Variables|Default|GetLightSwitchTimer)))))

(fn OnRep_DiscoLightsOn
  (bind _spotlight (Variables|Default|GetSpotLight))
  (Networking|SwitchHasAuthority
    (:Remote
      (if (Variables|Default|GetDiscoLightsOn)
        (CallFunction|LPFTransition true)
        (Development|PrintString "TurnOn" false false)
        (else
          (CallFunction|LPFTransition false)
          (Development|PrintString "TurnOff" false false)
          (Rendering|Components|Light|SetIntensity _spotlight 6e+01)
          (Rendering|Components|Light|SetLightColor _spotlight "(R=0.396755,G=0.278894,B=0.021219,A=1.000000)" true)
          (Rendering|Material|SetVectorParameterValue (Variables|Default|GetLampMID) "Emissive Multiplier" "(R=10.000000,G=6.200000,B=2.775000,A=1.000000)" :!class /Script/Engine.MaterialInstanceDynamic)
          (Rendering|Material|SetVectorParameterValue (Variables|Default|GetConeMID) "Emissive Color" "(R=1.000000,G=0.899035,B=0.624277,A=1.000000)" :!class /Script/Engine.MaterialInstanceDynamic))))))

(fn ConstructionScript
  (|CallParentFunction))
